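# Physics asset for secondary motion of the dangling pouches and gun pouch on the male medium Tiramisu player body cosmetic.
# Each of the eight front and back pouches and the gun pouch has a capsule body constrained to its own tiny MSTR anchor body.
# All joints share the same limited swing and tight twist.
# A right-thigh capsule is the only body outside these pairs and has no constraint, likely serving as a collider.
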PhysicsAsset /BRCosmetics/Characters/Player/Male/Medium/Bodies/M_MED_Tiramisu/Meshes/M_MED_Tiramisu_Dyn
bodies {
 dyn_pouch_bk_l_01 [{shapeType=Capsule, length=4.7307, radius=2.7605, center=<0.6181 0 1.3011>, rotation=<rot 90 0 0>}]
 dyn_pouch_bk_l_01_MSTR [{shapeType=Capsule, length=0.5, radius=0.5}]
 dyn_pouch_bk_l_02 [{shapeType=Capsule, length=4.7307, radius=2.7605, center=<0.6181 0 1.3011>, rotation=<rot 90 0 0>}]
 dyn_pouch_bk_l_02_MSTR [{shapeType=Capsule, length=0.5, radius=0.5}]
 dyn_pouch_bk_r_01 [{shapeType=Capsule, length=4.7307, radius=2.7605, center=<0.6181 0 1.3011>, rotation=<rot 90 0 0>}]
 dyn_pouch_bk_r_01_MSTR [{shapeType=Capsule, length=0.5, radius=0.5}]
 dyn_pouch_bk_r_02 [{shapeType=Capsule, length=4.7307, radius=2.7605, center=<0.6181 0 1.3011>, rotation=<rot 90 0 0>}]
 dyn_pouch_bk_r_02_MSTR [{shapeType=Capsule, length=0.5, radius=0.5}]
 dyn_pouch_fr_l_01 [{shapeType=Capsule, length=4.7307, radius=2.7605, center=<0.6181 0 1.3011>, rotation=<rot 90 0 0>}]
 dyn_pouch_fr_l_01_MSTR [{shapeType=Capsule, length=0.5, radius=0.5}]
 dyn_pouch_fr_l_02 [{shapeType=Capsule, length=4.7307, radius=2.7605, center=<0.6181 0 1.3011>, rotation=<rot 90 0 0>}]
 dyn_pouch_fr_l_02_MSTR [{shapeType=Capsule, length=0.5, radius=0.5}]
 dyn_pouch_fr_r_01 [{shapeType=Capsule, length=4.7307, radius=2.7605, center=<0.6181 0 1.3011>, rotation=<rot 90 0 0>}]
 dyn_pouch_fr_r_01_MSTR [{shapeType=Capsule, length=0.5, radius=0.5}]
 dyn_pouch_fr_r_02 [{shapeType=Capsule, length=4.7307, radius=2.7605, center=<0.6181 0 1.3011>, rotation=<rot 90 0 0>}]
 dyn_pouch_fr_r_02_MSTR [{shapeType=Capsule, length=0.5, radius=0.5}]
 dyn_pouch_gun [{shapeType=Capsule, length=11.6057, radius=6, center=<3.9092 0 5.8024>, rotation=<rot 84.5747 180 180>}]
 dyn_pouch_gun_MSTR [{shapeType=Capsule, length=0.5, radius=0.5}]
 thigh_r [{shapeType=Capsule, length=28.0711, radius=9.8724, center=<16.5433 -0.0001 -0.5953>, rotation=<rot 90 0 0>}]
}
constraints [
 {bone1Name=dyn_pouch_fr_l_01, bone2Name=dyn_pouch_fr_l_01_MSTR, swing1LimitDegrees=45, swing2LimitDegrees=45, twistLimitDegrees=5}
 {bone1Name=dyn_pouch_fr_l_02, bone2Name=dyn_pouch_fr_l_02_MSTR, swing1LimitDegrees=45, swing2LimitDegrees=45, twistLimitDegrees=5}
 {bone1Name=dyn_pouch_bk_l_01, bone2Name=dyn_pouch_bk_l_01_MSTR, swing1LimitDegrees=45, swing2LimitDegrees=45, twistLimitDegrees=5}
 {bone1Name=dyn_pouch_bk_l_02, bone2Name=dyn_pouch_bk_l_02_MSTR, swing1LimitDegrees=45, swing2LimitDegrees=45, twistLimitDegrees=5}
 {bone1Name=dyn_pouch_fr_r_01, bone2Name=dyn_pouch_fr_r_01_MSTR, swing1LimitDegrees=45, swing2LimitDegrees=45, twistLimitDegrees=5}
 {bone1Name=dyn_pouch_fr_r_02, bone2Name=dyn_pouch_fr_r_02_MSTR, swing1LimitDegrees=45, swing2LimitDegrees=45, twistLimitDegrees=5}
 {bone1Name=dyn_pouch_bk_r_01, bone2Name=dyn_pouch_bk_r_01_MSTR, swing1LimitDegrees=45, swing2LimitDegrees=45, twistLimitDegrees=5}
 {bone1Name=dyn_pouch_bk_r_02, bone2Name=dyn_pouch_bk_r_02_MSTR, swing1LimitDegrees=45, swing2LimitDegrees=45, twistLimitDegrees=5}
 {bone1Name=dyn_pouch_gun, bone2Name=dyn_pouch_gun_MSTR, swing1LimitDegrees=45, swing2LimitDegrees=45, twistLimitDegrees=5}
]
