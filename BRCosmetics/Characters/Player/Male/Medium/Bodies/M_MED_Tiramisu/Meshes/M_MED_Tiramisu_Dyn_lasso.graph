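# Physics asset for the dynamic lasso rope on the M_MED_Tiramisu medium male player body cosmetic.
# The two rope segments, each built from three capsules, form a jointed chain hanging from a small master body, dyn_rope_01_MSTR.
# Both joints set moderate swing limits and tight twist limits.
# An unconstrained capsule on the left thigh gives the leg a collision shape alongside the rope.
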PhysicsAsset /BRCosmetics/Characters/Player/Male/Medium/Bodies/M_MED_Tiramisu/Meshes/M_MED_Tiramisu_Dyn_lasso
bodies {
 dyn_rope_01 [
  {shapeType=Capsule, length=9.1994, radius=4.3542, center=<6.0921 0 4.704>, rotation=<rot 90 0 0>}
  {shapeType=Capsule, length=11.4182, radius=2.1355, center=<7.2982 6.6889 -0.2979>, rotation=<rot 87.1808 76.8323 59.0941>}
  {shapeType=Capsule, length=11.4182, radius=2.1355, center=<7.3385 -6.9613 0.6981>, rotation=<rot 82.3157 64.9371 79.9019>}
 ]
 dyn_rope_01_MSTR [{shapeType=Capsule, length=0.5, radius=0.5}]
 dyn_rope_02 [
  {shapeType=Capsule, length=9.1994, radius=4.3542, center=<6.0921 0 4.704>, rotation=<rot 90 0 0>}
  {shapeType=Capsule, length=11.4182, radius=2.1355, center=<7.1955 6.1181 0.4463>, rotation=<rot 77.0115 134.3795 155.4381>}
  {shapeType=Capsule, length=11.4182, radius=2.1355, center=<7.711 -6.6624 0.7309>, rotation=<rot 75.4868 164.3044 139.445>}
 ]
 thigh_l [{shapeType=Capsule, length=28.0711, radius=9.8724, center=<-16.5433 0.0001 0.5953>, rotation=<rot 90 0 0>}]
}
constraints [
 {bone1Name=dyn_rope_01, bone2Name=dyn_rope_01_MSTR, swing1LimitDegrees=45, swing2LimitDegrees=45, twistLimitDegrees=5}
 {bone1Name=dyn_rope_02, bone2Name=dyn_rope_01, swing1LimitDegrees=45, swing2LimitDegrees=45, twistLimitDegrees=5}
]
