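# Control Rig for the Tiramisu medium male player body that applies corrective offsets to shoulder deltoid fix bones as the upper arms rotate.
# Each upper arm's rotation relative to spine_05 becomes Euler angles, and absolute Z is remapped to a clamped weight.
# That weight blends additive local transforms onto one left and two right bones, including a front one.
# The sides are asymmetric, with different offsets and an inverted input range on the left.
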ControlRigBlueprint /BRCosmetics/Characters/Player/Male/Medium/Bodies/M_MED_Tiramisu/Meshes/Parts/M_MED_Tiramisu_ControlRig_shoulders
graphs [
 RigVMFunctionLibrary {links [], nodes []}
 RigVMModel {nodes [
   RigUnit_BeginExecution = &RigUnit_BeginExecution()
   n1 = &RigVMFunction_MathQuaternionToEuler(RotationOrder=ZYX, Value=&RigUnit_GetRelativeBoneTransform(Bone=upperarm_l, Space=spine_05).Transform.Rotation)
   /Script/ControlRig.RigUnit_ModifyBoneTransforms(
    BoneToModify=((Bone="dyn_shoulder_delt_fix_l_MSTR",Transform=(Rotation=(X=0,Y=0,Z=0,W=1),Translation=(X=0,Y=-5,Z=2),Scale3D=(X=1,Y=1,Z=1))))
    Mode=AdditiveLocal
    WeightMaximum=1
    Weight=&RigVMFunction_MathFloatRemap(
     SourceMaximum=50
     SourceMinimum=170
     TargetMinimum=0
     bClamp=true
     Value=/Script/RigVM.RigVMFunction_MathFloatAbs(Value=@n1.Result.Z) {is=struct RigVMUnitNode, template="Absolute::Execute(in Value,out Result)"}.Result
    ).Result
    BoneToModify.0.Transform.Rotation=&RigVMFunction_MathQuaternionFromEuler(Euler=/Script/RigVM.RigVMFunction_MathVectorAdd(A=(X=0,Y=0,Z=0), B=(X=0,Y=0,Z=0)) {is=struct RigVMUnitNode, template="Add::Execute(in A,in B,out Result)"}.Result).Result
    ExecutePin=@RigUnit_BeginExecution.ExecutePin
   ) {is=struct RigVMUnitNode}
   n2 = &RigUnit_GetRelativeBoneTransform(Bone=upperarm_r, Space=spine_05)
   n3 = &RigVMFunction_MathFloatRemap(
    SourceMaximum=90
    SourceMinimum=10
    TargetMinimum=0
    bClamp=true
    Value=&RigVMFunction_MathFloatAbs(Value=&RigVMFunction_MathQuaternionToEuler(RotationOrder=ZYX, Value=@n2.Transform.Rotation).Result.Z) {template="Absolute::Execute(in Value,out Result)"}.Result
   )
   > &RigUnit_ModifyBoneTransforms(
    BoneToModify=((Bone="dyn_shoulder_delt_fix_r_MSTR",Transform=(Rotation=(X=0,Y=0,Z=0,W=1),Translation=(X=2,Y=2,Z=2),Scale3D=(X=1,Y=1,Z=1))))
    Mode=AdditiveLocal
    WeightMaximum=1
    Weight=@n3.Result
    BoneToModify.0.Transform.Rotation=&RigVMFunction_MathQuaternionFromEuler(Euler=&RigVMFunction_MathVectorAdd(A=(X=0,Y=0,Z=0), B=(X=0,Y=0,Z=0)) {template="Add::Execute(in A,in B,out Result)"}.Result).Result
   )
   > &RigUnit_ModifyBoneTransforms(
    BoneToModify=((Bone="dyn_shoulder_delt_fix_fr_r_MSTR",Transform=(Rotation=(X=0,Y=0,Z=0,W=1),Translation=(X=10,Y=5,Z=0),Scale3D=(X=1,Y=1,Z=1))))
    Mode=AdditiveLocal
    WeightMaximum=1
    Weight=@n3.Result
    BoneToModify.0.Transform.Rotation=&RigVMFunction_MathQuaternionFromEuler(Euler=&RigVMFunction_MathVectorAdd(A=(X=0,Y=0,Z=0), B=(X=0,Y=0,Z=10)) {template="Add::Execute(in A,in B,out Result)"}.Result).Result
   )
  ]}
]
hierarchy [
 root {loc=<0 0 0>, rot=<0 0 0 1>}
 pelvis {parent=root, loc=<0 -0.7071 93.1191>, rot=<0.0127 0.707 -0.0127 -0.707>}
 spine_01 {parent=pelvis, loc=<6.7494 -0.6388 0>, rot=<0 0 0.034 -0.9994>}
 spine_02 {parent=spine_01, loc=<7.2518 -0.035 0>, rot=<0 0 0.0238 -0.9997>}
 spine_03 {parent=spine_02, loc=<6.0104 0 0>, rot=<0 0 0.0149 -0.9999>}
 spine_04 {parent=spine_03, loc=<8.1024 0 0>, rot=<0 0 -0.0132 -0.9999>}
 spine_05 {parent=spine_04, loc=<10.0146 0 0>, rot=<0 0 -0.0015 -1>}
 clavicle_l {parent=spine_05, loc=<4.3575 -0.1466 -2.3223>, rot=<0.0465 -0.7289 -0.009 -0.6829>}
 upperarm_l {parent=clavicle_l, loc=<15.5111 0 0>, rot=<0.0226 -0.3234 -0.0088 -0.9459>}
 lowerarm_l {parent=upperarm_l, loc=<26.6514 0 0>, rot=<0.0391 -0.0099 -0.2453 -0.9686>}
 upperarm_twist_01_l {parent=upperarm_l, loc=<0.0041 0 0>, rot=<0 0 0 -1>}
 deform_deltoid_fr_l {parent=upperarm_twist_01_l, loc=<2.4511 0 0>, rot=<0 0 0 -1>}
 dyn_shoulder_delt_fix_l_MSTR {parent=upperarm_twist_01_l, loc=<6.2629 -6.5148 2.2678>, rot=<-0.3817 -0.0099 -0.1035 -0.9184>}
 dyn_shoulder_delt_fix_l {parent=dyn_shoulder_delt_fix_l_MSTR, loc=<0.3 0 0>, rot=<0 0 0 -1>}
 upperarm_twist_02_l {parent=upperarm_l, loc=<13.7941 0 0>, rot=<0 0 0 -1>}
 deform_pectoral_fr_l {parent=clavicle_l, loc=<9.3905 0 0>, rot=<0 0 0 1>}
 clavicle_r {parent=spine_05, loc=<4.3575 -0.1466 2.3223>, rot=<0.7289 0.0465 -0.6829 0.009>}
 upperarm_r {parent=clavicle_r, loc=<-15.5111 0 0>, rot=<0.0226 -0.3234 -0.0088 -0.9459>}
 lowerarm_r {parent=upperarm_r, loc=<-26.6514 0 0>, rot=<0.0391 -0.0099 -0.2453 -0.9686>}
 upperarm_twist_01_r {parent=upperarm_r, loc=<-0.0041 0 0>, rot=<0 0 0 1>}
 deform_deltoid_fr_r {parent=upperarm_twist_01_r, loc=<-2.4512 0 0>, rot=<0 0 0 1>}
 dyn_shoulder_delt_fix_r_MSTR {parent=upperarm_twist_01_r, loc=<-6.2631 6.5147 -2.2676>, rot=<0.1035 -0.9184 -0.3817 0.0099>}
 dyn_shoulder_delt_fix_r {parent=dyn_shoulder_delt_fix_r_MSTR, loc=<0.2997 -0.0001 0.0002>, rot=<0 0 0 -1>}
 dyn_shoulder_delt_fix_fr_r_MSTR {parent=upperarm_twist_01_r, loc=<-2.9075 -1.9246 -4.4151>, rot=<0.222 -0.9616 0.1612 -0.0002>}
 dyn_shoulder_delt_fix_fr_r {parent=dyn_shoulder_delt_fix_fr_r_MSTR, loc=<0.2998 -0.0001 0.0002>, rot=<0 0 0 -1>}
 upperarm_twist_02_r {parent=upperarm_r, loc=<-13.7942 0 0>, rot=<0 0 0 1>}
 deform_pectoral_fr_r {parent=clavicle_r, loc=<-9.39 0 0>, rot=<0 0 0 -1>}
 neck_01 {parent=spine_05, loc=<11.3138 0.1529 0>, rot=<0 0 -0.2003 -0.9797>}
 neck_02 {parent=neck_01, loc=<4.5627 0 0>, rot=<0 0 0.0504 -0.9987>}
 head {parent=neck_02, loc=<5.3914 0 0>, rot=<0 0 0.115 -0.9934>}
 dyn_widowsPeak_01_MSTR {parent=head, loc=<10.8639 8.5682 0>, rot=<0.2174 0.6728 0.6728 0.2174>}
 dyn_widowsPeak_01 {parent=dyn_widowsPeak_01_MSTR, loc=<0.3 0 0>, rot=<0 0 0 -1>}
 dyn_widowsPeak_02 {parent=dyn_widowsPeak_01, loc=<6.3327 0 0>, rot=<0 -0.1516 0 -0.9884>}
 dyn_hood_back_top_MSTR {parent=head, loc=<8.4434 -5.3923 0>, rot=<0.2445 -0.6635 0.6635 -0.2445>}
 dyn_hood_back_top {parent=dyn_hood_back_top_MSTR, loc=<0.3 0 0>, rot=<0 0 0 1>}
 backpack_root {parent=spine_05, loc=<9.0735 -9.9874 0>, rot=<0.0284 -0.7065 -0.0284 -0.7065>}
 cape_mid_01_l_MSTR {parent=backpack_root, loc=<6.333 1.3484 -1.7132>, rot=<-0.4589 -0.442 0.5323 -0.5575>}
 cape_mid_01_l {parent=cape_mid_01_l_MSTR, loc=<0.3 0 0>, rot=<0 0 0 1>}
 cape_mid_02_l {parent=cape_mid_01_l, loc=<15.6774 0 -0.1591>, rot=<0 -0.0565 0 -0.9984>}
 cape_mid_03_l {parent=cape_mid_02_l, loc=<22.7491 0 -0.028>, rot=<0 -0.0363 0 -0.9993>}
 cape_mid_04_l {parent=cape_mid_03_l, loc=<20.0287 0 0>, rot=<0 0 0 1>}
 cape_mid_05_l {parent=cape_mid_04_l, loc=<18.4294 0 0>, rot=<0 0 0 1>}
 cape_mid_06_l {parent=cape_mid_05_l, loc=<13.3811 0 0>, rot=<0 0 0 1>}
 cape_mid_01_r_MSTR {parent=backpack_root, loc=<-6.3332 1.3484 -1.7136>, rot=<-0.442 -0.4589 0.5575 -0.5323>}
 cape_mid_01_r {parent=cape_mid_01_r_MSTR, loc=<0.3 0 0>, rot=<0 0 0 1>}
 cape_mid_02_r {parent=cape_mid_01_r, loc=<15.6767 0 -0.1589>, rot=<0 -0.0565 0 -0.9984>}
 cape_mid_03_r {parent=cape_mid_02_r, loc=<22.7496 0 -0.0281>, rot=<0 -0.0363 0 -0.9993>}
 cape_mid_04_r {parent=cape_mid_03_r, loc=<20.0287 0 0>, rot=<0 0 0 1>}
 cape_mid_05_r {parent=cape_mid_04_r, loc=<18.4294 0 0.0001>, rot=<0 0 0 1>}
 cape_mid_06_r {parent=cape_mid_05_r, loc=<13.3811 0 0>, rot=<0 0 0 1>}
 clavicle_cape_l {parent=spine_05, loc=<4.3578 -0.1465 -2.3223>, rot=<0.0465 -0.7289 -0.009 -0.6829>}
 cape_01_l {parent=clavicle_cape_l, loc=<14.2451 -3.5936 6.4534>, rot=<-0.2333 -0.2002 0.6154 -0.7258>}
 cape_02_l {parent=cape_01_l, loc=<6.9515 -0.0004 -1.0637>, rot=<-0.012 -0.3639 -0.0268 -0.931>}
 cape_03_l {parent=cape_02_l, loc=<12.4341 0 -0.1591>, rot=<0 -0.0565 0 -0.9984>}
 cape_04_l {parent=cape_03_l, loc=<23.9025 0 -0.028>, rot=<0 -0.0363 0 -0.9993>}
 cape_05_l {parent=cape_04_l, loc=<20.6415 0 0>, rot=<0 0 0 1>}
 cape_06_l {parent=cape_05_l, loc=<19.0579 0 0>, rot=<0 0 0 1>}
 cape_07_l {parent=cape_06_l, loc=<14.1389 0 0>, rot=<0 0 0 1>}
 clavicle_cape_r {parent=spine_05, loc=<4.3578 -0.1465 2.3223>, rot=<0.6829 0.009 0.7289 -0.0465>}
 cape_01_r {parent=clavicle_cape_r, loc=<14.245 3.8111 6.1639>, rot=<0.2333 -0.2002 -0.6154 -0.7258>}
 cape_02_r {parent=cape_01_r, loc=<6.9512 0.0003 -1.0632>, rot=<0.012 -0.3639 0.0268 -0.931>}
 cape_03_r {parent=cape_02_r, loc=<12.4345 0 -0.1592>, rot=<0 -0.0565 0 -0.9984>}
 cape_04_r {parent=cape_03_r, loc=<23.9027 0.0001 -0.028>, rot=<0 -0.0363 0 -0.9993>}
 cape_05_r {parent=cape_04_r, loc=<20.6413 -0.0001 0>, rot=<0 0 0 1>}
 cape_06_r {parent=cape_05_r, loc=<19.0579 0.0001 -0.0001>, rot=<0 0 0 1>}
 cape_07_r {parent=cape_06_r, loc=<14.1389 0 0>, rot=<0 0 0 1>}
 dyn_medallion_l_MSTR {parent=spine_05, loc=<1.2513 11.7671 -8.1059>, rot=<-0.018 -0.612 0.3065 -0.7288>}
 dyn_medallion_l {parent=dyn_medallion_l_MSTR, loc=<0.3 0 0>, rot=<0 0 0 1>}
 dyn_medallion_r_MSTR {parent=spine_05, loc=<1.2511 11.7671 8.1059>, rot=<0.7288 -0.3065 0.612 0.018>}
 dyn_medallion_r {parent=dyn_medallion_r_MSTR, loc=<0.3 -0.0002 0.0003>, rot=<0 0 0 -1>}
 thigh_l {parent=pelvis, loc=<-2.5693 -0.0109 -7.7501>, rot=<0.0012 0.0572 0.0155 -0.9982>}
 calf_l {parent=thigh_l, loc=<-39.0484 0 0>, rot=<0.0006 -0.0161 -0.0391 -0.9991>}
 thigh_r {parent=pelvis, loc=<-2.5691 -0.0109 7.7501>, rot=<0.0572 -0.0012 0.9982 0.0155>}
 calf_r {parent=thigh_r, loc=<39.0487 0 0>, rot=<0.0006 -0.0161 -0.0391 -0.9991>}
 dyn_rope_01_MSTR {parent=pelvis, loc=<4.9992 0.4035 -15.6012>, rot=<-0.0155 -0.9977 0.0014 -0.0657>}
 dyn_rope_01 {parent=dyn_rope_01_MSTR, loc=<0.3 0 0>, rot=<0 0 0 -1>}
 dyn_rope_02 {parent=dyn_rope_01, loc=<13.0364 0 0>, rot=<0 0 0 -1>}
 dyn_crouch_collision {parent=root, loc=<0 0 0>, rot=<0 0 0 1>}
 dyn_root_simspace {parent=root, loc=<0 0 0>, rot=<0 0 0 1>}
 dyn_root_slope {parent=root, loc=<0 0 0>, rot=<0 0 0 1>}
]
