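# Physics asset for the cape on the Tiramisu male medium player cosmetic, with body colliders and two rope bodies off the pelvis.
# Four cape capsule chains hang from clavicle and backpack anchors, using limited swing and tight twist.
# Free-rotation cross-links tie neighbouring chains at intervals, making them move like a sheet.
# Torso and limb capsules, a crouch sphere and a broad root slope box provide collision shapes, unconstrained in the latter two.
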PhysicsAsset /BRCosmetics/Characters/Player/Male/Medium/Bodies/M_MED_Tiramisu/Meshes/Parts/M_MED_Tiramisu_FaceAcc_Dyn_cape
bodies {
 backpack_root [{shapeType=Capsule, length=0.505, radius=0.505, rotation=<rot 0 -90 0>}]
 calf_l [{shapeType=Capsule, length=31.8125, radius=12.5363, center=<-31.4163 -2.6359 -0.0003>, rotation=<rot 90 0 -1.5716>}]
 calf_r [{shapeType=Capsule, length=31.8125, radius=12.5363, center=<31.4163 2.6359 0.0003>, rotation=<rot 90 0 -1.5716>}]
 cape_01_l [{shapeType=Capsule, length=0, radius=0.5}]
 cape_01_r [{shapeType=Capsule, length=0, radius=0.5}]
 cape_02_l [{shapeType=Capsule, length=15.2188, radius=15, center=<12.0264 4.2599 14.2339>, rotation=<rot -90 -0.0297 0.0297>}]
 cape_02_r [{shapeType=Capsule, length=15.2188, radius=15, center=<12.0264 -4.2599 14.2339>, rotation=<rot -90 -0.0297 0.0297>}]
 cape_03_l [
  {shapeType=Capsule, length=15.2188, radius=15, center=<12.0267 4.5975 15>, rotation=<rot -90 0.0099 -0.0099>}
  {shapeType=Box, extentX=50, extentY=41.875, extentZ=7.5, center=<12.6087 -9.8682 2.463>, rotation=<rot -4.4122 0.211 -2.7412>}
 ]
 cape_03_r [{shapeType=Capsule, length=15.2188, radius=15, center=<12.0267 -4.5975 14.684>, rotation=<rot -90 0.0099 -0.0099>}]
 cape_04_l [{shapeType=Capsule, length=15.8438, radius=15, center=<11.2039 0 15>, rotation=<rot -90 -0.0264 0.0264>}]
 cape_04_r [{shapeType=Capsule, length=13.2812, radius=15, center=<9.1357 0.0001 15>, rotation=<rot -90 0.0049 -0.0049>}]
 cape_05_l [{shapeType=Capsule, length=10.9375, radius=15, center=<8.8549 -0.0001 15>, rotation=<rot -90 0.0049 -0.0049>}]
 cape_05_r [{shapeType=Capsule, length=10.9375, radius=15, center=<8.8549 -0.0001 15>, rotation=<rot -90 0.0049 -0.0049>}]
 cape_06_l [{shapeType=Capsule, length=10.9375, radius=15, center=<6.2228 -0.0001 15>, rotation=<rot -90 179.9802 -179.9802>}]
 cape_06_r [{shapeType=Capsule, length=10.9375, radius=15, center=<6.2228 -0.0001 15>, rotation=<rot -90 179.9802 -179.9802>}]
 cape_07_l [{shapeType=Capsule, length=10.9375, radius=15, center=<6.6116 -0.0003 15>, rotation=<rot -90 -0.0099 0.0099>}]
 cape_07_r [{shapeType=Capsule, length=10.9375, radius=15, center=<6.6116 -0.0003 15>, rotation=<rot -90 -0.0099 0.0099>}]
 cape_mid_01_l [{shapeType=Capsule, length=13.2812, radius=15, center=<9.6403 0 15>, rotation=<rot -90 0.0396 -0.0396>}]
 cape_mid_01_l_MSTR [{shapeType=Capsule, length=0, radius=0.5}]
 cape_mid_01_r [{shapeType=Capsule, length=13.2812, radius=15, center=<9.6403 0 15>, rotation=<rot -90 0.0396 -0.0396>}]
 cape_mid_01_r_MSTR [{shapeType=Capsule, length=0, radius=0.5}]
 cape_mid_02_l [{shapeType=Capsule, length=15.2188, radius=15, center=<11.4454 0 15>, rotation=<rot -90 0.0099 -0.0099>}]
 cape_mid_02_r [{shapeType=Capsule, length=13.2812, radius=15, center=<9.6403 0 15>, rotation=<rot -90 0.0396 -0.0396>}]
 cape_mid_03_l [{shapeType=Capsule, length=15.8438, radius=15, center=<10.7205 0 15>, rotation=<rot -90 -0.0099 0.0099>}]
 cape_mid_03_r [{shapeType=Capsule, length=13.2812, radius=15, center=<8.6523 0 15>, rotation=<rot -90 -0.0297 0.0297>}]
 cape_mid_04_l [{shapeType=Capsule, length=10.9375, radius=15, center=<8.6896 -0.0001 15>, rotation=<rot -90 -0.0198 0.0198>}]
 cape_mid_04_r [{shapeType=Capsule, length=10.9375, radius=15, center=<8.6896 -0.0001 15>, rotation=<rot -90 -0.0198 0.0198>}]
 cape_mid_05_l [{shapeType=Capsule, length=10.9375, radius=15, center=<6.2193 -0.0002 15>, rotation=<rot -90 0.0049 -0.0049>}]
 cape_mid_05_r [{shapeType=Capsule, length=10.9375, radius=15, center=<6.2193 -0.0002 15>, rotation=<rot -90 0.0049 -0.0049>}]
 cape_mid_06_l [{shapeType=Capsule, length=10.9375, radius=15, center=<5.5138 -0.0001 15>, rotation=<rot -90 0.0396 -0.0396>}]
 cape_mid_06_r [{shapeType=Capsule, length=10.9375, radius=15, center=<5.5138 -0.0001 15>, rotation=<rot -90 0.0396 -0.0396>}]
 clavicle_cape_l [{shapeType=Capsule, length=0.505, radius=0.505, rotation=<rot 0 -90 0>}]
 clavicle_cape_r [{shapeType=Capsule, length=0.505, radius=0.505, rotation=<rot 0 -90 0>}]
 clavicle_l [{shapeType=Capsule, length=8, radius=9, center=<10 -3 0>, rotation=<rot 90 10 0>}]
 clavicle_r [{shapeType=Capsule, length=8, radius=9, center=<-10 3 0>, rotation=<rot 90 10 0>}]
 dyn_crouch_collision [{shapeType=Sphere, radius=53.3125, center=<-0.0006 5.0355 3.6223>}]
 dyn_root_slope [{shapeType=Box, extentX=231.0625, extentY=231.0625, extentZ=12.375, center=<0 0 -4.1865>}]
 dyn_rope_01 [{shapeType=Capsule, length=10.625, radius=10.0363, center=<6.8808 -0.1787 -0.0005>, rotation=<rot 90 0 0>}]
 dyn_rope_02 [{shapeType=Capsule, length=10.625, radius=10.0363, center=<6.8808 -0.1787 -0.0005>, rotation=<rot 90 0 0>}]
 lowerarm_l [{shapeType=Capsule, length=27, radius=8, center=<1 0 0>, rotation=<rot 90 0 0>}]
 lowerarm_r [
  {shapeType=Capsule, length=27, radius=8, center=<1 0 0>, rotation=<rot 90 0 0>}
  {shapeType=Capsule, length=23.5938, radius=6.4062, center=<-31.9441 -1.2158 5.2412>, rotation=<rot 0 0 89.9994>}
 ]
 pelvis [
  {shapeType=Capsule, length=11.1562, radius=16.6925, center=<-5.3453 0.0725 -0.0005>}
  {shapeType=Capsule, length=17.6562, radius=16.7238, center=<-21.8171 2.1268 0.0008>}
 ]
 spine_01 [{shapeType=Capsule, length=9.875, radius=17.9738, center=<0 1.2622 -0.0003>}]
 spine_02 [{shapeType=Capsule, length=15.375, radius=12.4738}]
 spine_03 [{shapeType=Capsule, length=15.2188, radius=14, center=<0 -2.1065 -0.0001>}]
 spine_04 [{shapeType=Capsule, length=16.0625, radius=14, center=<1.0291 -4 0.0002>}]
 spine_05 [{shapeType=Capsule, length=11.75, radius=14, center=<-2.9499 -4 -0.0007>}]
 thigh_l [{shapeType=Capsule, length=16.125, radius=18.5988, center=<-20.1925 3.5421 3.1216>, rotation=<rot 90 0 0>}]
 thigh_r [{shapeType=Capsule, length=16.125, radius=18.5988, center=<20.1925 -4.0916 -2.7025>, rotation=<rot 90 0 0>}]
 upperarm_l [
  {shapeType=Capsule, length=20, radius=10, center=<13 -3 0>, rotation=<rot 90 0 0>}
  {shapeType=Capsule, length=8.875, radius=7.875, center=<5.7273 -0.0002 0.1463>, rotation=<rot -83.203 -180 180>}
 ]
 upperarm_r [
  {shapeType=Capsule, length=20, radius=10, center=<-13 2 0>, rotation=<rot 90 0 0>}
  {shapeType=Capsule, length=8.875, radius=7.875, center=<-5.7273 0.0002 -0.1463>, rotation=<rot -83.203 -180 180>}
 ]
}
constraints [
 {bone1Name=cape_02_l, bone2Name=cape_01_l, swing1LimitDegrees=45, swing2LimitDegrees=45, twistLimitDegrees=5}
 {bone1Name=cape_03_l, bone2Name=cape_02_l, swing1LimitDegrees=45, swing2LimitDegrees=45, twistLimitDegrees=5}
 {bone1Name=cape_04_l, bone2Name=cape_03_l, swing1LimitDegrees=45, swing2LimitDegrees=45, twistLimitDegrees=5}
 {bone1Name=cape_05_l, bone2Name=cape_04_l, swing1LimitDegrees=45, swing2LimitDegrees=45, twistLimitDegrees=5}
 {bone1Name=cape_06_l, bone2Name=cape_05_l, swing1LimitDegrees=45, swing2LimitDegrees=45, twistLimitDegrees=5}
 {bone1Name=cape_07_l, bone2Name=cape_06_l, swing1LimitDegrees=45, swing2LimitDegrees=45, twistLimitDegrees=5}
 {bone1Name=cape_01_l, bone2Name=clavicle_cape_l, swing1LimitDegrees=45, swing2LimitDegrees=45, twistLimitDegrees=5}
 {bone1Name=cape_01_r, bone2Name=clavicle_cape_r, swing1LimitDegrees=45, swing2LimitDegrees=45, twistLimitDegrees=5}
 {bone1Name=cape_02_r, bone2Name=cape_01_r, swing1LimitDegrees=10, swing2LimitDegrees=45, twistLimitDegrees=5}
 {bone1Name=cape_03_r, bone2Name=cape_02_r, swing1LimitDegrees=45, swing2LimitDegrees=45, twistLimitDegrees=5}
 {bone1Name=cape_04_r, bone2Name=cape_03_r, swing1LimitDegrees=45, swing2LimitDegrees=45, twistLimitDegrees=5}
 {bone1Name=cape_05_r, bone2Name=cape_04_r, swing1LimitDegrees=45, swing2LimitDegrees=45, twistLimitDegrees=5}
 {bone1Name=cape_06_r, bone2Name=cape_05_r, swing1LimitDegrees=45, swing2LimitDegrees=45, twistLimitDegrees=5}
 {bone1Name=cape_07_r, bone2Name=cape_06_r, swing1LimitDegrees=45, swing2LimitDegrees=45, twistLimitDegrees=5}
 {bone1Name=cape_mid_01_l, bone2Name=cape_mid_01_l_MSTR, swing1LimitDegrees=5, swing2LimitDegrees=45, twistLimitDegrees=5}
 {bone1Name=cape_mid_02_l, bone2Name=cape_mid_01_l, swing1LimitDegrees=45, swing2LimitDegrees=45, twistLimitDegrees=5}
 {bone1Name=cape_mid_03_l, bone2Name=cape_mid_02_l, swing1LimitDegrees=45, swing2LimitDegrees=45, twistLimitDegrees=5}
 {bone1Name=cape_mid_04_l, bone2Name=cape_mid_03_l, swing1LimitDegrees=45, swing2LimitDegrees=45, twistLimitDegrees=5}
 {bone1Name=cape_mid_05_l, bone2Name=cape_mid_04_l, swing1LimitDegrees=45, swing2LimitDegrees=45, twistLimitDegrees=5}
 {bone1Name=cape_mid_06_l, bone2Name=cape_mid_05_l, swing1LimitDegrees=45, swing2LimitDegrees=45, twistLimitDegrees=5}
 {bone1Name=cape_mid_01_r, bone2Name=cape_mid_01_r_MSTR, swing1LimitDegrees=5, swing2LimitDegrees=45, twistLimitDegrees=5}
 {bone1Name=cape_mid_02_r, bone2Name=cape_mid_01_r, swing1LimitDegrees=45, swing2LimitDegrees=45, twistLimitDegrees=5}
 {bone1Name=cape_mid_03_r, bone2Name=cape_mid_02_r, swing1LimitDegrees=45, swing2LimitDegrees=45, twistLimitDegrees=5}
 {bone1Name=cape_mid_04_r, bone2Name=cape_mid_03_r, swing1LimitDegrees=45, swing2LimitDegrees=45, twistLimitDegrees=5}
 {bone1Name=cape_mid_05_r, bone2Name=cape_mid_04_r, swing1LimitDegrees=45, swing2LimitDegrees=45, twistLimitDegrees=5}
 {bone1Name=cape_mid_06_r, bone2Name=cape_mid_05_r, swing1LimitDegrees=45, swing2LimitDegrees=45, twistLimitDegrees=5}
 {
  bone1Name=clavicle_cape_l
  bone2Name=clavicle_l
  swing1LimitDegrees=45
  swing1Motion=Free
  swing2LimitDegrees=45
  swing2Motion=Free
  twistLimitDegrees=45
  twistMotion=Free
 }
 {
  bone1Name=clavicle_cape_r
  bone2Name=clavicle_r
  swing1LimitDegrees=45
  swing1Motion=Free
  swing2LimitDegrees=45
  swing2Motion=Free
  twistLimitDegrees=45
  twistMotion=Free
 }
 {
  bone1Name=clavicle_l
  bone2Name=spine_05
  swing1LimitDegrees=45
  swing1Motion=Free
  swing2LimitDegrees=45
  swing2Motion=Free
  twistLimitDegrees=45
  twistMotion=Free
 }
 {
  bone1Name=clavicle_r
  bone2Name=spine_05
  swing1LimitDegrees=45
  swing1Motion=Free
  swing2LimitDegrees=45
  swing2Motion=Free
  twistLimitDegrees=45
  twistMotion=Free
 }
 {
  bone1Name=backpack_root
  bone2Name=spine_05
  swing1LimitDegrees=45
  swing1Motion=Free
  swing2LimitDegrees=45
  swing2Motion=Free
  twistLimitDegrees=45
  twistMotion=Free
 }
 {
  bone1Name=spine_05
  bone2Name=spine_04
  swing1LimitDegrees=45
  swing1Motion=Free
  swing2LimitDegrees=45
  swing2Motion=Free
  twistLimitDegrees=45
  twistMotion=Free
 }
 {
  bone1Name=spine_04
  bone2Name=spine_03
  swing1LimitDegrees=45
  swing1Motion=Free
  swing2LimitDegrees=45
  swing2Motion=Free
  twistLimitDegrees=45
  twistMotion=Free
 }
 {
  bone1Name=spine_03
  bone2Name=spine_02
  swing1LimitDegrees=45
  swing1Motion=Free
  swing2LimitDegrees=45
  swing2Motion=Free
  twistLimitDegrees=45
  twistMotion=Free
 }
 {
  bone1Name=spine_02
  bone2Name=spine_01
  swing1LimitDegrees=45
  swing1Motion=Free
  swing2LimitDegrees=45
  swing2Motion=Free
  twistLimitDegrees=45
  twistMotion=Free
 }
 {
  bone1Name=spine_01
  bone2Name=pelvis
  swing1LimitDegrees=45
  swing1Motion=Free
  swing2LimitDegrees=45
  swing2Motion=Free
  twistLimitDegrees=45
  twistMotion=Free
 }
 {
  bone1Name=lowerarm_l
  bone2Name=upperarm_l
  swing1LimitDegrees=45
  swing1Motion=Free
  swing2LimitDegrees=45
  swing2Motion=Free
  twistLimitDegrees=45
  twistMotion=Free
 }
 {
  bone1Name=upperarm_l
  bone2Name=clavicle_l
  swing1LimitDegrees=45
  swing1Motion=Free
  swing2LimitDegrees=45
  swing2Motion=Free
  twistLimitDegrees=45
  twistMotion=Free
 }
 {
  bone1Name=lowerarm_r
  bone2Name=upperarm_r
  swing1LimitDegrees=45
  swing1Motion=Free
  swing2LimitDegrees=45
  swing2Motion=Free
  twistLimitDegrees=45
  twistMotion=Free
 }
 {
  bone1Name=upperarm_r
  bone2Name=clavicle_r
  swing1LimitDegrees=45
  swing1Motion=Free
  swing2LimitDegrees=45
  swing2Motion=Free
  twistLimitDegrees=45
  twistMotion=Free
 }
 {
  bone1Name=thigh_l
  bone2Name=pelvis
  swing1LimitDegrees=45
  swing1Motion=Free
  swing2LimitDegrees=45
  swing2Motion=Free
  twistLimitDegrees=45
  twistMotion=Free
 }
 {
  bone1Name=calf_l
  bone2Name=thigh_l
  swing1LimitDegrees=45
  swing1Motion=Free
  swing2LimitDegrees=45
  swing2Motion=Free
  twistLimitDegrees=45
  twistMotion=Free
 }
 {
  bone1Name=thigh_r
  bone2Name=pelvis
  swing1LimitDegrees=45
  swing1Motion=Free
  swing2LimitDegrees=45
  swing2Motion=Free
  twistLimitDegrees=45
  twistMotion=Free
 }
 {
  bone1Name=calf_r
  bone2Name=thigh_r
  swing1LimitDegrees=45
  swing1Motion=Free
  swing2LimitDegrees=45
  swing2Motion=Free
  twistLimitDegrees=45
  twistMotion=Free
 }
 {
  bone1Name=cape_03_r
  bone2Name=cape_mid_02_r
  swing1LimitDegrees=45
  swing1Motion=Free
  swing2LimitDegrees=45
  swing2Motion=Free
  twistLimitDegrees=45
  twistMotion=Free
 }
 {
  bone1Name=cape_03_l
  bone2Name=cape_mid_02_l
  swing1LimitDegrees=45
  swing1Motion=Free
  swing2LimitDegrees=45
  swing2Motion=Free
  twistLimitDegrees=45
  twistMotion=Free
 }
 {
  bone1Name=cape_mid_04_l
  bone2Name=cape_mid_04_r
  swing1LimitDegrees=45
  swing1Motion=Free
  swing2LimitDegrees=45
  swing2Motion=Free
  twistLimitDegrees=45
  twistMotion=Free
 }
 {
  bone1Name=cape_05_r
  bone2Name=cape_mid_04_r
  swing1LimitDegrees=45
  swing1Motion=Free
  swing2LimitDegrees=45
  swing2Motion=Free
  twistLimitDegrees=45
  twistMotion=Free
 }
 {
  bone1Name=cape_05_l
  bone2Name=cape_mid_04_l
  swing1LimitDegrees=45
  swing1Motion=Free
  swing2LimitDegrees=45
  swing2Motion=Free
  twistLimitDegrees=45
  twistMotion=Free
 }
 {
  bone1Name=cape_04_r
  bone2Name=cape_mid_03_r
  swing1LimitDegrees=45
  swing1Motion=Free
  swing2LimitDegrees=45
  swing2Motion=Free
  twistLimitDegrees=45
  twistMotion=Free
 }
 {
  bone1Name=cape_07_r
  bone2Name=cape_mid_06_r
  swing1LimitDegrees=45
  swing1Motion=Free
  swing2LimitDegrees=45
  swing2Motion=Free
  twistLimitDegrees=45
  twistMotion=Free
 }
 {
  bone1Name=cape_07_l
  bone2Name=cape_mid_06_l
  swing1LimitDegrees=45
  swing1Motion=Free
  swing2LimitDegrees=45
  swing2Motion=Free
  twistLimitDegrees=45
  twistMotion=Free
 }
 {
  bone1Name=cape_mid_03_l
  bone2Name=cape_mid_03_r
  swing1LimitDegrees=45
  swing1Motion=Free
  swing2LimitDegrees=45
  swing2Motion=Free
  twistLimitDegrees=45
  twistMotion=Free
 }
 {
  bone1Name=cape_mid_06_l
  bone2Name=cape_mid_06_r
  swing1LimitDegrees=45
  swing1Motion=Free
  swing2LimitDegrees=45
  swing2Motion=Free
  twistLimitDegrees=45
  twistMotion=Free
 }
 {
  bone1Name=dyn_rope_02
  bone2Name=dyn_rope_01
  swing1LimitDegrees=45
  swing1Motion=Free
  swing2LimitDegrees=45
  swing2Motion=Free
  twistLimitDegrees=45
  twistMotion=Free
 }
 {
  bone1Name=dyn_rope_01
  bone2Name=pelvis
  swing1LimitDegrees=45
  swing1Motion=Free
  swing2LimitDegrees=45
  swing2Motion=Free
  twistLimitDegrees=45
  twistMotion=Free
 }
 {
  bone1Name=cape_mid_01_l_MSTR
  bone2Name=backpack_root
  swing1LimitDegrees=45
  swing1Motion=Free
  swing2LimitDegrees=45
  swing2Motion=Free
  twistLimitDegrees=45
  twistMotion=Free
 }
 {
  bone1Name=cape_mid_01_r_MSTR
  bone2Name=backpack_root
  swing1LimitDegrees=45
  swing1Motion=Free
  swing2LimitDegrees=45
  swing2Motion=Free
  twistLimitDegrees=45
  twistMotion=Free
 }
]
