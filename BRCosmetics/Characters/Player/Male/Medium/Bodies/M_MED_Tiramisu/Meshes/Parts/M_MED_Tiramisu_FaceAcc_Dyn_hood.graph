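# Physics asset for the dynamic hood on the Tiramisu medium male character's face accessory mesh, defining bodies for secondary motion.
# Capsule bodies form two chains, each rooted at a tiny MSTR capsule: a single back-top hood segment and a two-segment widow's peak.
# Angular limits let the hood back swing widely with little twist, while each widow's peak joint stays tightly limited on every axis.
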PhysicsAsset /BRCosmetics/Characters/Player/Male/Medium/Bodies/M_MED_Tiramisu/Meshes/Parts/M_MED_Tiramisu_FaceAcc_Dyn_hood
bodies {
 dyn_hood_back_top [{shapeType=Capsule, length=4.6022, radius=3.5725, center=<2.8448 0 0>, rotation=<rot 90 0 0>}]
 dyn_hood_back_top_MSTR [{shapeType=Capsule, length=0.505, radius=0.505}]
 dyn_widowsPeak_01 [{shapeType=Capsule, length=4.6022, radius=3.5725, center=<2.8448 0 0>, rotation=<rot 90 0 0>}]
 dyn_widowsPeak_01_MSTR [{shapeType=Capsule, length=0.505, radius=0.505}]
 dyn_widowsPeak_02 [{shapeType=Capsule, length=4.6022, radius=3.5725, center=<2.8448 0 0>, rotation=<rot 90 0 0>}]
}
constraints [
 {bone1Name=dyn_widowsPeak_01, bone2Name=dyn_widowsPeak_01_MSTR, swing1LimitDegrees=5, swing2LimitDegrees=5, twistLimitDegrees=5}
 {bone1Name=dyn_widowsPeak_02, bone2Name=dyn_widowsPeak_01, swing1LimitDegrees=5, swing2LimitDegrees=5, twistLimitDegrees=5}
 {bone1Name=dyn_hood_back_top, bone2Name=dyn_hood_back_top_MSTR, swing1LimitDegrees=45, swing2LimitDegrees=45, twistLimitDegrees=5}
]
